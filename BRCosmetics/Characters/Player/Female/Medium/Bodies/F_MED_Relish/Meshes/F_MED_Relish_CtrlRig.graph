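# Control Rig for the F_MED_Relish character body that procedurally adjusts the shoulder pauldron bones as the arms move.
# Bone rotations become Euler angles, then clamped remaps set the weights of transform offsets.
# Left clavicle Y drives dyn_pauldron_l, while right upperarm Y and Z drive dyn_pauldron_r, with Y also driving its tweak bone.
# A second left pauldron variant sits unconnected to execution, and other dyn bones are untouched.
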ControlRigBlueprint /BRCosmetics/Characters/Player/Female/Medium/Bodies/F_MED_Relish/Meshes/F_MED_Relish_CtrlRig
graphs [
 RigVMFunctionLibrary {links [], nodes []}
 RigVMModel {nodes [
   RigUnit_BeginExecution = &RigUnit_BeginExecution()
   n1 = &RigUnit_GetTransform(Item=(Type=Bone,Name="clavicle_l"))
   &RigUnit_ModifyTransforms(
    ItemToModify=((Item=(Type=Bone),Transform=(Rotation=(X=0,Y=0,Z=0,W=1),Translation=(X=0,Y=0,Z=0),Scale3D=(X=1,Y=1,Z=1))),(Item=(Type=Bone,Name="dyn_pauldron_l"),Transform=(Rotation=(X=0,Y=0,Z=0,W=1),Translation=(X=0,Y=1.5,Z=3.5),Scale3D=(X=1,Y=1.1,Z=1))))
    Weight=&RigVMFunction_MathFloatRemap(
     SourceMaximum=-50
     SourceMinimum=0
     TargetMinimum=0
     bClamp=true
     Value=&RigVMFunction_MathQuaternionToEuler(RotationOrder=XYZ, Value=@n1.Transform.Rotation).Result.Y
    ).Result
    ItemToModify.1.Transform.Rotation=&RigVMFunction_MathQuaternionFromEuler(Euler=(X=0,Y=10,Z=32), RotationOrder=XYZ).Result
    ExecutePin=@RigUnit_BeginExecution.ExecutePin
   )
   n2 = &RigUnit_GetTransform(Item=(Type=Bone,Name="upperarm_r"))
   > &RigUnit_ModifyTransforms(
    ItemToModify=((Item=(Type=Bone),Transform=(Rotation=(X=0,Y=0,Z=0,W=1),Translation=(X=0,Y=0,Z=0),Scale3D=(X=1,Y=1,Z=1))),(Item=(Type=Bone,Name="dyn_pauldron_r"),Transform=(Rotation=(X=0,Y=0,Z=0,W=1),Translation=(X=0,Y=1,Z=-1),Scale3D=(X=1,Y=1.1,Z=1))))
    Weight=&RigVMFunction_MathFloatRemap(
     SourceMaximum=60
     SourceMinimum=-80
     TargetMinimum=0
     bClamp=true
     Value=&RigVMFunction_MathQuaternionToEuler(RotationOrder=XYZ, Value=@n2.Transform.Rotation).Result.Y
    ).Result
    ItemToModify.1.Transform.Rotation=&RigVMFunction_MathQuaternionFromEuler(Euler=(X=0,Y=0,Z=25), RotationOrder=XYZ).Result
   )
   n3 = &RigUnit_GetTransform(Item=(Type=Bone,Name="upperarm_r"))
   > &RigUnit_ModifyTransforms(
    ItemToModify=((Item=(Type=Bone),Transform=(Rotation=(X=0,Y=0,Z=0,W=1),Translation=(X=0,Y=0,Z=0),Scale3D=(X=1,Y=1,Z=1))),(Item=(Type=Bone,Name="dyn_pauldron_r_tweak"),Transform=(Rotation=(X=0,Y=0,Z=0,W=1),Translation=(X=0,Y=0,Z=0),Scale3D=(X=1,Y=1,Z=1))))
    Weight=&RigVMFunction_MathFloatRemap(
     SourceMaximum=40
     SourceMinimum=-25
     TargetMinimum=0
     bClamp=true
     Value=&RigVMFunction_MathQuaternionToEuler(RotationOrder=XYZ, Value=@n3.Transform.Rotation).Result.Y
    ).Result
    ItemToModify.1.Transform.Rotation=&RigVMFunction_MathQuaternionFromEuler(Euler=(X=-20,Y=0,Z=-90), RotationOrder=XYZ).Result
   )
   n4 = &RigUnit_GetTransform(Item=(Type=Bone,Name="upperarm_r"))
   > &RigUnit_ModifyTransforms(
    ItemToModify=((Item=(Type=Bone),Transform=(Rotation=(X=0,Y=0,Z=0,W=1),Translation=(X=0,Y=0,Z=0),Scale3D=(X=1,Y=1,Z=1))),(Item=(Type=Bone,Name="dyn_pauldron_r"),Transform=(Rotation=(X=0,Y=0,Z=0,W=1),Translation=(X=0,Y=0,Z=-1.5),Scale3D=(X=1,Y=1,Z=1))))
    Weight=&RigVMFunction_MathFloatRemap(
     SourceMaximum=50
     SourceMinimum=-20
     TargetMinimum=0
     bClamp=true
     Value=&RigVMFunction_MathQuaternionToEuler(RotationOrder=XYZ, Value=@n4.Transform.Rotation).Result.Z
    ).Result
    ItemToModify.1.Transform.Rotation=&RigVMFunction_MathQuaternionFromEuler(Euler=(X=0,Y=0,Z=0), RotationOrder=XYZ).Result
   )
   n5 = &RigUnit_GetTransform(Item=(Type=Bone,Name="clavicle_l"))
   &RigUnit_ModifyTransforms(
    ExecutePin=()
    ItemToModify=((Item=(Type=Bone),Transform=(Rotation=(X=0,Y=0,Z=0,W=1),Translation=(X=0,Y=0,Z=0),Scale3D=(X=1,Y=1,Z=1))),(Item=(Type=Bone,Name="dyn_pauldron_l"),Transform=(Rotation=(X=0,Y=0,Z=0,W=1),Translation=(X=-1.5,Y=1.5,Z=3.5),Scale3D=(X=1,Y=1.1,Z=1))))
    Weight=&RigVMFunction_MathFloatRemap(
     SourceMaximum=-50
     SourceMinimum=0
     TargetMinimum=0
     bClamp=true
     Value=&RigVMFunction_MathQuaternionToEuler(RotationOrder=XYZ, Value=@n5.Transform.Rotation).Result.Y
    ).Result
    ItemToModify.1.Transform.Rotation=&RigVMFunction_MathQuaternionFromEuler(Euler=(X=15,Y=20,Z=32), RotationOrder=XYZ).Result
   )
  ]}
]
hierarchy [
 root {loc=<0 0 0>, rot=<0 0 0 1>}
 pelvis {parent=root, loc=<0 -0.5301 98.0571>, rot=<0.0127 0.707 -0.0127 -0.707>}
 spine_01 {parent=pelvis, loc=<6.1208 -0.9328 0>, rot=<0 0 0.034 -0.9994>}
 spine_02 {parent=spine_01, loc=<6.5996 -0.2095 0>, rot=<0 0 0.0238 -0.9997>}
 spine_03 {parent=spine_02, loc=<6.4914 -0.1237 0>, rot=<0 0 0.0149 -0.9999>}
 spine_04 {parent=spine_03, loc=<6.9598 0.0794 0>, rot=<0 0 -0.0132 -0.9999>}
 spine_05 {parent=spine_04, loc=<6.9987 0 0>, rot=<0 0 -0.0015 -1>}
 clavicle_l {parent=spine_05, loc=<5.4831 -0.1053 -3.2946>, rot=<0.0465 -0.7289 -0.009 -0.6829>}
 upperarm_l {parent=clavicle_l, loc=<9.8658 -1.2136 -0.4105>, rot=<0.0164 -0.3245 -0.0269 -0.9454>}
 lowerarm_l {parent=upperarm_l, loc=<26.465 -0.8345 0.259>, rot=<0.0036 0.0161 -0.2298 -0.9731>}
 hand_l {parent=lowerarm_l, loc=<22.7355 -0.3203 -0.6106>, rot=<-0.6784 -0.0223 0.0735 -0.7307>}
 index_metacarpal_l {parent=hand_l, loc=<1.6535 0.0183 -2.96>, rot=<0.0158 -0.0124 -0.0173 -0.9996>}
 index_01_l {parent=index_metacarpal_l, loc=<8.3008 0.0357 -0.0524>, rot=<-0.0044 0.0485 0.0924 -0.9945>}
 index_02_l {parent=index_01_l, loc=<3.3587 -0.3004 0.0452>, rot=<-0.002 0.0027 0.3225 -0.9466>}
 index_03_l {parent=index_02_l, loc=<2.5578 -0.123 -0.0252>, rot=<-0.002 0.0067 0.2558 -0.9667>}
 middle_metacarpal_l {parent=hand_l, loc=<2.1856 -0.1725 -0.5984>, rot=<0.006 0.0168 -0.0329 -0.9993>}
 middle_01_l {parent=middle_metacarpal_l, loc=<7.3275 -0.2447 0.0129>, rot=<-0.0208 0.0042 0.0858 -0.9961>}
 middle_02_l {parent=middle_01_l, loc=<4.0123 0.1401 0.1175>, rot=<-0.0218 0.0258 0.3512 -0.9357>}
 middle_03_l {parent=middle_02_l, loc=<2.6077 0.0534 0.1373>, rot=<0.0082 0.0188 0.1708 -0.9851>}
 pinky_metacarpal_l {parent=hand_l, loc=<1.7616 -0.9753 3.3814>, rot=<-0.0397 0.0814 -0.069 -0.9935>}
 pinky_01_l {parent=pinky_metacarpal_l, loc=<7.1709 -0.4831 -0.1094>, rot=<-0.0654 -0.0035 0.2242 -0.9723>}
 pinky_02_l {parent=pinky_01_l, loc=<2.8274 0.0893 -0.1043>, rot=<0.0037 0.0011 0.2854 -0.9584>}
 pinky_03_l {parent=pinky_02_l, loc=<2.2046 0.0307 -0.0078>, rot=<-0.0002 0.0098 0.1445 -0.9895>}
 ring_metacarpal_l {parent=hand_l, loc=<2.2074 -0.3625 1.3638>, rot=<-0.0284 0.0722 -0.0458 -0.9959>}
 ring_01_l {parent=ring_metacarpal_l, loc=<7.0295 -0.6197 -0.1446>, rot=<-0.0489 -0.0252 0.1892 -0.9804>}
 ring_02_l {parent=ring_01_l, loc=<3.6174 0.4647 -0.1027>, rot=<0 0 0.2806 -0.9598>}
 ring_03_l {parent=ring_02_l, loc=<2.4152 0.114 0.0199>, rot=<0 0 0.1189 -0.9929>}
 thumb_01_l {parent=hand_l, loc=<1.3118 -0.3191 -2.459>, rot=<0.6546 -0.4042 0.0664 -0.6354>}
 thumb_02_l {parent=thumb_01_l, loc=<5.9336 -0.0448 0.0401>, rot=<-0.0037 0.0213 0.298 -0.9543>}
 thumb_03_l {parent=thumb_02_l, loc=<2.2223 0.3682 -0.0907>, rot=<0.0118 -0.013 0.0066 -0.9998>}
 lowerarm_twist_01_l {parent=lowerarm_l, loc=<18.316 0 0>, rot=<0 0 0 1>}
 lowerarm_twist_02_l {parent=lowerarm_l, loc=<10.6707 0 0>, rot=<0 0 0 1>}
 upperarm_twist_01_l {parent=upperarm_l, loc=<7.2135 -0.1134 0>, rot=<0 0 0 1>}
 upperarm_twist_02_l {parent=upperarm_l, loc=<19.7998 -0.1134 0>, rot=<0 0 0 1>}
 dyn_light {parent=clavicle_l, loc=<8.4499 5.9366 -2.811>, rot=<-0.0381 0.2692 -0.4831 -0.8323>}
 dyn_pauldron_l {parent=clavicle_l, loc=<3.5228 -1.1877 3.507>, rot=<0.785 -0.0226 0.6172 -0.0484>}
 clavicle_r {parent=spine_05, loc=<5.4831 -0.1053 3.2946>, rot=<0.7289 0.0465 -0.6829 0.009>}
 upperarm_r {parent=clavicle_r, loc=<-9.8658 1.2136 0.4104>, rot=<0.0164 -0.3245 -0.0269 -0.9454>}
 lowerarm_r {parent=upperarm_r, loc=<-26.4651 0.8345 -0.259>, rot=<0.0036 0.0161 -0.2298 -0.9731>}
 hand_r {parent=lowerarm_r, loc=<-22.7356 0.3203 0.6106>, rot=<-0.6784 -0.0223 0.0735 -0.7307>}
 index_metacarpal_r {parent=hand_r, loc=<-1.6534 -0.0181 2.9601>, rot=<0.0158 -0.0124 -0.0173 -0.9996>}
 index_01_r {parent=index_metacarpal_r, loc=<-8.3008 -0.0357 0.0524>, rot=<-0.0044 0.0485 0.0924 -0.9945>}
 index_02_r {parent=index_01_r, loc=<-3.3589 0.3003 -0.0453>, rot=<-0.002 0.0027 0.3225 -0.9466>}
 index_03_r {parent=index_02_r, loc=<-2.5578 0.123 0.0252>, rot=<-0.002 0.0067 0.2558 -0.9667>}
 middle_metacarpal_r {parent=hand_r, loc=<-2.1856 0.1726 0.5984>, rot=<0.006 0.0168 -0.0329 -0.9993>}
 middle_01_r {parent=middle_metacarpal_r, loc=<-7.3273 0.2447 -0.0129>, rot=<-0.0208 0.0042 0.0858 -0.9961>}
 middle_02_r {parent=middle_01_r, loc=<-4.0124 -0.14 -0.1175>, rot=<-0.0218 0.0258 0.3512 -0.9357>}
 middle_03_r {parent=middle_02_r, loc=<-2.6077 -0.0533 -0.1373>, rot=<0.0082 0.0188 0.1708 -0.9851>}
 pinky_metacarpal_r {parent=hand_r, loc=<-1.7617 0.9753 -3.3814>, rot=<-0.0397 0.0814 -0.069 -0.9935>}
 pinky_01_r {parent=pinky_metacarpal_r, loc=<-7.1707 0.4832 0.1094>, rot=<-0.0654 -0.0035 0.2242 -0.9723>}
 pinky_02_r {parent=pinky_01_r, loc=<-2.8274 -0.0892 0.1043>, rot=<0.0037 0.0011 0.2854 -0.9584>}
 pinky_03_r {parent=pinky_02_r, loc=<-2.2045 -0.031 0.0078>, rot=<-0.0002 0.0098 0.1445 -0.9895>}
 ring_metacarpal_r {parent=hand_r, loc=<-2.2073 0.3627 -1.3638>, rot=<-0.0284 0.0722 -0.0458 -0.9959>}
 ring_01_r {parent=ring_metacarpal_r, loc=<-7.0297 0.6197 0.1446>, rot=<-0.0489 -0.0252 0.1892 -0.9804>}
 ring_02_r {parent=ring_01_r, loc=<-3.6171 -0.4648 0.1027>, rot=<0 0 0.2806 -0.9598>}
 ring_03_r {parent=ring_02_r, loc=<-2.4153 -0.1141 -0.0199>, rot=<0 0 0.1189 -0.9929>}
 thumb_01_r {parent=hand_r, loc=<-1.3119 0.3193 2.459>, rot=<0.6546 -0.4042 0.0664 -0.6354>}
 thumb_02_r {parent=thumb_01_r, loc=<-5.9335 0.0447 -0.0401>, rot=<-0.0037 0.0213 0.298 -0.9543>}
 thumb_03_r {parent=thumb_02_r, loc=<-2.2223 -0.3682 0.0907>, rot=<0.0118 -0.013 0.0066 -0.9998>}
 lowerarm_twist_01_r {parent=lowerarm_r, loc=<-18.3161 0 0>, rot=<0 0 0 1>}
 lowerarm_twist_02_r {parent=lowerarm_r, loc=<-10.6707 0 0>, rot=<0 0 0 1>}
 upperarm_twist_01_r {parent=upperarm_r, loc=<-7.2135 0.1134 0>, rot=<0 0 0 1>}
 upperarm_twist_02_r {parent=upperarm_r, loc=<-19.7998 0.1134 0>, rot=<0 0 0 1>}
 dyn_pauldron_r {parent=clavicle_r, loc=<-4.0665 1.1621 -3.5446>, rot=<-0.0224 -0.7851 -0.0483 -0.6171>}
 dyn_pauldron_r_tweak {parent=dyn_pauldron_r, loc=<2.3016 -0.4272 -2.9562>, rot=<0.4709 -0.8783 -0.0697 -0.0443>}
 neck_01 {parent=spine_05, loc=<11.3138 0.1529 0>, rot=<0 0 -0.2003 -0.9797>}
 neck_02 {parent=neck_01, loc=<4.5627 0 0>, rot=<0 0 0.0504 -0.9987>}
 head {parent=neck_02, loc=<5.3914 0 0>, rot=<0 0 0.115 -0.9934>}
 dyn_arm_r_holster {parent=spine_04, loc=<-3.1435 -2.8955 10.4343>, rot=<-0.003 -0.2698 0.9627 0.0196>}
 dyn_arm_l_holster {parent=spine_04, loc=<-3.1045 -3.9513 -9.9799>, rot=<0.0452 -0.9443 0.3145 -0.0856>}
 thigh_l {parent=pelvis, loc=<-1.0855 -0.2109 -8.7053>, rot=<0.0009 0.0355 0.0155 -0.9993>}
 calf_l {parent=thigh_l, loc=<-42.5284 -0.0019 -0.7297>, rot=<0.0006 -0.0161 -0.0391 -0.9991>}
 calf_twist_01_l {parent=calf_l, loc=<-35.8483 -1.9197 -0.5794>, rot=<0.0237 0.0095 -0.0035 -0.9997>}
 dyn_l_shoeString_MSTR {parent=calf_twist_01_l, loc=<3.4143 4.6888 -0.2942>, rot=<0.3673 0.6242 0.6711 -0.1583>}
 dyn_l_shoeLoop_01 {parent=dyn_l_shoeString_MSTR, loc=<0 0 0>, rot=<0 0 0 -1>}
 dyn_l_shoeLoop_02 {parent=dyn_l_shoeLoop_01, loc=<3.2864 -0.1142 -0.2285>, rot=<0.024 0.0208 0.0366 -0.9988>}
 dyn_l_shoeLoop_03 {parent=dyn_l_shoeString_MSTR, loc=<0 0 0>, rot=<-0.0434 -0.0592 0.6379 -0.7666>}
 dyn_l_shoeLoop_04 {parent=dyn_l_shoeLoop_03, loc=<3.0628 -0.098 -0.2382>, rot=<0.024 0.0208 0.0366 -0.9988>}
 dyn_l_shoeString_03 {parent=dyn_l_shoeString_MSTR, loc=<0.2352 -0.4118 -0.0187>, rot=<-0.014 0.0064 0.5118 -0.8589>}
 dyn_l_shoeString_04 {parent=dyn_l_shoeString_03, loc=<2.6828 -0.0705 -0.2547>, rot=<0.024 0.0208 0.0366 -0.9988>}
 dyn_l_shoeString_01 {parent=dyn_l_shoeString_MSTR, loc=<1.0746 0.2585 0.4445>, rot=<0.0133 -0.0175 0.2564 -0.9663>}
 dyn_l_shoeString_02 {parent=dyn_l_shoeString_01, loc=<2.6828 -0.0705 -0.2547>, rot=<-0.0645 -0.0394 0.0416 -0.9963>}
 calf_twist_02_l {parent=calf_l, loc=<-19.1546 -1.7943 -0.2638>, rot=<0.0237 0.0095 -0.0035 -0.9997>}
 foot_l {parent=calf_l, loc=<-46.7401 -2.5014 -0.592>, rot=<0.0427 -0.0238 0.0353 -0.9982>}
 ball_l {parent=foot_l, loc=<-3.5658 10.8066 -0.1544>, rot=<0.0001 -0.0001 -0.7063 -0.7079>}
 thigh_twist_01_l {parent=thigh_l, loc=<-27.952 0 0>, rot=<0.0013 0 0 -1>}
 thigh_r {parent=pelvis, loc=<-1.0853 -0.2109 8.7053>, rot=<0.0355 -0.0009 0.9993 0.0155>}
 calf_r {parent=thigh_r, loc=<42.5286 0.0019 0.7297>, rot=<0.0006 -0.0161 -0.0391 -0.9991>}
 calf_twist_01_r {parent=calf_r, loc=<35.8485 1.9197 0.5793>, rot=<0.0237 0.0095 -0.0035 -0.9997>}
 dyn_r_shoeString_MSTR {parent=calf_twist_01_r, loc=<-3.513 -4.6278 0.2489>, rot=<-0.6266 -0.1515 0.3318 -0.6887>}
 dyn_r_shoeLoop_01 {parent=dyn_r_shoeString_MSTR, loc=<0 0 0>, rot=<0 0 0 -1>}
 dyn_r_shoeLoop_02 {parent=dyn_r_shoeLoop_01, loc=<3.8141 -0.1605 -0.2007>, rot=<0.0252 -0.0092 -0.0166 -0.9995>}
 dyn_r_shoeLoop_03 {parent=dyn_r_shoeString_MSTR, loc=<0 0 0>, rot=<0 0 -0.6244 -0.7811>}
 dyn_r_shoeLoop_04 {parent=dyn_r_shoeLoop_03, loc=<3.8141 -0.1605 -0.2007>, rot=<0.0252 -0.0092 -0.0166 -0.9995>}
 dyn_r_shoeString_01 {parent=dyn_r_shoeString_MSTR, loc=<1.2308 -0.3385 0.4163>, rot=<-0.0324 0.0186 -0.2947 -0.9549>}
 dyn_r_shoeString_02 {parent=dyn_r_shoeString_01, loc=<2.3173 -0.2094 -0.172>, rot=<0.0255 -0.0084 0.012 -0.9996>}
 dyn_r_shoeString_03 {parent=dyn_r_shoeString_MSTR, loc=<0.7827 0.9733 0.1067>, rot=<-0.0294 0.0357 -0.5233 -0.8509>}
 dyn_r_shoeString_04 {parent=dyn_r_shoeString_03, loc=<2.3173 -0.2094 -0.172>, rot=<0.0252 -0.0092 -0.0166 -0.9995>}
 calf_twist_02_r {parent=calf_r, loc=<19.1546 1.7943 0.2638>, rot=<0.0237 0.0095 -0.0035 -0.9997>}
 foot_r {parent=calf_r, loc=<46.7403 2.5014 0.5919>, rot=<0.0427 -0.0238 0.0353 -0.9982>}
 ball_r {parent=foot_r, loc=<3.5658 -10.8066 0.1544>, rot=<0.0001 -0.0001 -0.7063 -0.7079>}
 thigh_twist_01_r {parent=thigh_r, loc=<27.952 0 0>, rot=<0.0013 0 0 -1>}
 dyn_thigh_holster {parent=thigh_twist_01_r, loc=<-4.7185 2.7044 8.4225>, rot=<0.0357 0.5899 -0.0272 -0.8063>}
 dyn_thigh_holster_strap1 {parent=dyn_thigh_holster, loc=<1.0534 -2.4081 8.9255>, rot=<0.5497 0.3608 0.6978 -0.2842>}
 dyn_thigh_holster_strap3 {parent=dyn_thigh_holster, loc=<3.8211 -5.0477 9.2424>, rot=<0.3849 0.6132 0.4745 -0.5007>}
 dyn_MSTR {parent=pelvis, loc=<0 0 0>, rot=<-0.0127 -0.707 0.0127 -0.707>}
 dyn_belt_pouch1 {parent=dyn_MSTR, loc=<12.5659 4.862 4.8842>, rot=<0.238 -0.591 -0.2373 -0.7334>}
 dyn_belt_pouch2 {parent=dyn_MSTR, loc=<13.2007 -2.8199 6.598>, rot=<-0.0178 -0.5297 0.1784 -0.829>}
 dyn_belt_pouch3 {parent=dyn_MSTR, loc=<9.2865 -8.2749 7.3659>, rot=<-0.2423 -0.472 0.3871 -0.7541>}
 dyn_belt_pouch4 {parent=dyn_MSTR, loc=<4.6179 -10.9134 4.3391>, rot=<-0.3034 -0.3432 0.5887 -0.666>}
 dyn_belt_pouch5 {parent=dyn_MSTR, loc=<4.4482 -10.8109 7.8169>, rot=<-0.3454 -0.4179 0.5354 -0.6477>}
 dyn_belt_pouch6 {parent=dyn_MSTR, loc=<-5.5198 -10.6328 4.9877>, rot=<-0.4322 -0.2895 0.6633 -0.538>}
 dyn_belt_pouch7 {parent=dyn_MSTR, loc=<-10.4658 -7.7899 6.3499>, rot=<-0.5035 -0.2108 0.7453 -0.3827>}
 dyn_belt_pouch8 {parent=dyn_MSTR, loc=<-13.6084 -4.6036 3.6018>, rot=<-0.557 -0.0572 0.8029 -0.2043>}
 dyn_belt_pouch9 {parent=dyn_MSTR, loc=<-11.3125 5.8025 4.7016>, rot=<-0.584 0.2751 0.7193 0.2566>}
 dyn_belt_radio {parent=dyn_MSTR, loc=<-13.9657 0.045 6.1773>, rot=<-0.589 0.0095 0.808 0.013>}
]
