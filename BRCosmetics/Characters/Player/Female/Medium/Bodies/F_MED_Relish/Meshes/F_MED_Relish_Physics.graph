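# Physics asset for the F_MED_Relish female medium player body, giving secondary motion to gear bones such as belt pouches, radio, light and holsters.
# Capsules on pelvis, upper spine and thighs form body shapes linked by free-swinging constraints, while gear bones use capsules, plus a box for the thigh holster.
# Tightly angle-limited constraints hang belt items from the pelvis, arm holsters and light from spine_04, and holster straps from the right-thigh holster.
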PhysicsAsset /BRCosmetics/Characters/Player/Female/Medium/Bodies/F_MED_Relish/Meshes/F_MED_Relish_Physics
bodies {
 dyn_arm_l_holster [{shapeType=Capsule, length=5, radius=2, shapeName=dyn_arm_l_holster_capsule, center=<5 0 1.4253>, rotation=<rot 0 0 90>}]
 dyn_arm_r_holster [{shapeType=Capsule, length=5, radius=2, shapeName=dyn_arm_r_holster_capsule, center=<6.0323 0 1.5>, rotation=<rot 0 0 90>}]
 dyn_belt_pouch1 [
  {
   shapeType=Capsule
   length=4.6821
   radius=2.5855
   shapeName=dyn_belt_pouch1_capsule
   center=<5.2171 -0.0205 1.6273>
   rotation=<rot 88.5765 -90.2073 -0.0008>
  }
  {shapeType=Capsule, length=4.1973, radius=2.4455, shapeName=dyn_belt_pouch1_capsule, center=<0.2062 -0.3498 2.0663>, rotation=<rot 0 0 90>}
 ]
 dyn_belt_pouch2 [{shapeType=Capsule, length=6, radius=2.5, shapeName=dyn_belt_pouch2_capsule, center=<4 0 1>, rotation=<rot 90 0 0>}]
 dyn_belt_pouch3 [{shapeType=Capsule, length=8, radius=2.5, shapeName=dyn_belt_pouch3_capsule, center=<5 0 1>, rotation=<rot 90 0 0>}]
 dyn_belt_pouch4 [{shapeType=Capsule, length=2, radius=2, shapeName=dyn_belt_pouch4_capsule, center=<3 0 0.5>, rotation=<rot 90 0 0>}]
 dyn_belt_pouch5 [{shapeType=Capsule, length=2, radius=1, shapeName=dyn_belt_pouch5_capsule, center=<2 0 0.5>, rotation=<rot 90 0 0>}]
 dyn_belt_pouch6 [{shapeType=Capsule, length=2, radius=2, shapeName=dyn_belt_pouch6_capsule, center=<1 0 1>, rotation=<rot 90 0 0>}]
 dyn_belt_pouch7 [{shapeType=Capsule, length=8, radius=2.5, shapeName=dyn_belt_pouch7_capsule, center=<5 0 1.2>, rotation=<rot 90 0 0>}]
 dyn_belt_pouch8 [{shapeType=Capsule, length=12, radius=2, shapeName=dyn_belt_pouch8_capsule, center=<7 0 0.5>, rotation=<rot 90 0 0>}]
 dyn_belt_pouch9 [{shapeType=Capsule, length=1.6217, radius=1.6189, shapeName=dyn_belt_pouch9_capsule, center=<3.2731 -0.5553 3.5883>, rotation=<rot 90 0 0>}]
 dyn_belt_radio [{shapeType=Capsule, length=5, radius=2.5, shapeName=dyn_belt_radio_capsule, center=<1.3651 0 0.5>, rotation=<rot 90 0 0>}]
 dyn_light [{shapeType=Capsule, length=3, radius=2, shapeName=dyn_light_capsule, center=<1.5 0 -0.5>, rotation=<rot 0 0 90>}]
 dyn_thigh_holster [{
   shapeType=Box
   extentX=10.3892
   extentY=5.1492
   extentZ=22.348
   shapeName=dyn_thigh_holster_box
   center=<1.761 0 0.4719>
   rotation=<rot 0 89.9999 15.0007>
  }]
 dyn_thigh_holster_strap1 [{shapeType=Capsule, length=2, radius=2, shapeName=dyn_thigh_holster_strap1_capsule, center=<1.2 0 0>, rotation=<rot 0 0 90>}]
 dyn_thigh_holster_strap3 [{shapeType=Capsule, length=3, radius=1, shapeName=dyn_thigh_holster_strap3_capsule, center=<0.5 0 -0.112>, rotation=<rot 90 90 85>}]
 pelvis [{shapeType=Capsule, length=7.0808, radius=9.8552, shapeName=pelvis_capsule, center=<0.2315 -1.177 0.0001>}]
 spine_04 [{shapeType=Capsule, length=16.0345, radius=10.4098, shapeName=spine_04_capsule, center=<-6.7704 0.6632 0>, rotation=<rot 90 0 0>}]
 thigh_l [{
   shapeType=Capsule
   length=6.4125
   radius=7.3759
   shapeName=thigh_l_capsule
   center=<-3.4667 -0.8468 0.6435>
   rotation=<rot 90 179.5524 -177.5894>
  }]
 thigh_twist_01_l [{
   shapeType=Capsule
   length=19.4993
   radius=6.8094
   shapeName=thigh_twist_01_r_capsule
   center=<8 -0.0002 -0.2035>
   rotation=<rot 90 -179.5253 0.4752>
  }]
 thigh_twist_01_r [{
   shapeType=Capsule
   length=27.2998
   radius=7
   shapeName=thigh_twist_01_r_capsule
   center=<-8 0.0002 0.2035>
   rotation=<rot 90 -177.4698 -177.4692>
  }]
}
constraints [
 {
  bone1Name=spine_04
  bone2Name=pelvis
  swing1LimitDegrees=45
  swing1Motion=Free
  swing2LimitDegrees=45
  swing2Motion=Free
  twistLimitDegrees=45
  twistMotion=Free
 }
 {bone1Name=dyn_belt_pouch1, bone2Name=pelvis, swing1LimitDegrees=5, swing2LimitDegrees=20, twistLimitDegrees=20}
 {bone1Name=dyn_belt_pouch2, bone2Name=pelvis, swing1LimitDegrees=1, swing2LimitDegrees=12, twistLimitDegrees=1}
 {bone1Name=dyn_belt_pouch3, bone2Name=pelvis, swing1LimitDegrees=1, swing2LimitDegrees=12, twistLimitDegrees=1}
 {bone1Name=dyn_belt_pouch4, bone2Name=pelvis, swing1LimitDegrees=1, swing2LimitDegrees=12, twistLimitDegrees=1}
 {bone1Name=dyn_belt_pouch5, bone2Name=pelvis, swing1LimitDegrees=1, swing2LimitDegrees=12, twistLimitDegrees=1}
 {bone1Name=dyn_belt_pouch6, bone2Name=pelvis, swing1LimitDegrees=1, swing2LimitDegrees=12, twistLimitDegrees=1}
 {bone1Name=dyn_belt_pouch7, bone2Name=pelvis, swing1LimitDegrees=1, swing2LimitDegrees=12, twistLimitDegrees=1}
 {bone1Name=dyn_belt_pouch8, bone2Name=pelvis, swing1LimitDegrees=1, swing2LimitDegrees=12, twistLimitDegrees=1}
 {bone1Name=dyn_belt_pouch9, bone2Name=pelvis, swing1LimitDegrees=5, swing2LimitDegrees=12, twistLimitDegrees=1}
 {bone1Name=dyn_belt_radio, bone2Name=pelvis, swing1LimitDegrees=1, swing2LimitDegrees=12, twistLimitDegrees=1}
 {bone1Name=dyn_arm_r_holster, bone2Name=spine_04, swing1LimitDegrees=1, swing2LimitDegrees=10, twistLimitDegrees=1}
 {bone1Name=dyn_arm_l_holster, bone2Name=spine_04, swing1LimitDegrees=1, swing2LimitDegrees=10, twistLimitDegrees=1}
 {bone1Name=dyn_light, bone2Name=spine_04, swing1LimitDegrees=1, swing2LimitDegrees=10, twistLimitDegrees=1}
 {bone1Name=dyn_thigh_holster, bone2Name=thigh_twist_01_r, swing1LimitDegrees=1, swing2LimitDegrees=5, twistLimitDegrees=1}
 {bone1Name=dyn_thigh_holster_strap1, bone2Name=dyn_thigh_holster, swing1LimitDegrees=1, swing2LimitDegrees=20, twistLimitDegrees=1}
 {bone1Name=dyn_thigh_holster_strap3, bone2Name=dyn_thigh_holster, swing1LimitDegrees=1, swing2LimitDegrees=20, twistLimitDegrees=1}
 {
  bone1Name=thigh_twist_01_r
  bone2Name=pelvis
  swing1LimitDegrees=45
  swing1Motion=Free
  swing2LimitDegrees=45
  swing2Motion=Free
  twistLimitDegrees=45
  twistMotion=Free
 }
 {
  bone1Name=thigh_l
  bone2Name=pelvis
  swing1LimitDegrees=45
  swing1Motion=Free
  swing2LimitDegrees=45
  swing2Motion=Free
  twistLimitDegrees=45
  twistMotion=Free
 }
 {
  bone1Name=thigh_twist_01_l
  bone2Name=thigh_l
  swing1LimitDegrees=45
  swing1Motion=Free
  swing2LimitDegrees=45
  swing2Motion=Free
  twistLimitDegrees=45
  twistMotion=Free
 }
]
