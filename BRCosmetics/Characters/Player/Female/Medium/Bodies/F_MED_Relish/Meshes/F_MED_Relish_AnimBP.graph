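# Post-process style AnimBP for the Relish female body cosmetic that copies the parent mesh pose and fixes up gear bones.
# A RigidBody sim drives dynamic bones, then offsets fade in while the owner's Relish idle montage plays, and the result is cached.
# Bool blends pick additive pauldron offsets for parachuting or skydiving, or the plain pose for emotes.
# Otherwise a Control Rig runs, with belt pouch offsets added when crouching.
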AnimBlueprint /BRCosmetics/Characters/Player/Female/Medium/Bodies/F_MED_Relish/Meshes/F_MED_Relish_AnimBP
target_skeleton=&/F_MED_Relish_Skeleton
graphs [
 AnimGraph {nodes [
   n1 = "Transform (Modify) Bone - Bone: dyn_belt_pouch1"(
    ComponentPose=Local To Component(LocalPose=Use cached pose 'cachedPose'())
    Alpha=1
    Translation="1.493,0.000,2.655"
    Rotation="-17.251,0.000,-0.000"
    Scale="0.8,0.8,0.8"
   ) {
    rotationMode=BMM_Additive
    scaleMode=BMM_Additive
    scaleSpace=BCS_BoneSpace
    translationMode=BMM_Additive
    alphaBoolBlend {blendOption=Linear}
    alphaScaleBiasClamp {bInterpResult=true}
    boneToModify {boneName=dyn_belt_pouch1}
    rotation {pitch=-17.251}
    scale=<0.8 0.8 0.8>
    translation {x=1.493, z=2.655}
   }
   n2 = "Transform (Modify) Bone - Bone: dyn_belt_pouch9"(ComponentPose=@n1, Alpha=1, Translation="-1,0,3", Rotation="-10,15,-20", Scale="1,1,1") {
    rotationMode=BMM_Additive
    translationMode=BMM_Additive
    alphaBoolBlend {blendOption=Linear}
    boneToModify {boneName=dyn_belt_pouch9}
    rotation=<rot -10 15 -20>
    translation {x=-1, z=3}
   }
   n3 = Control Rig(Source=Blend Poses by bool(BlendPose_0=Component To Local(ComponentPose=@n2), BlendPose_1=Use cached pose 'cachedPose'(), bActiveValue=Get bIsCrouching()) {blendType=Linear}) {
    alphaBoolBlend {blendOption=Linear}
    controlRigAssetReference {blueprintRigClass=&/F_MED_Relish_CtrlRig.F_MED_Relish_CtrlRig_C}
   }
   n4 = Blend Poses by bool(BlendPose_0=Use cached pose 'cachedPose'(), BlendPose_1=@n3, bActiveValue=Get bIsPlayingEmote()) {blendType=Linear}
   n5 = "Transform (Modify) Bone - Bone: dyn_pauldron_l"(
    ComponentPose=Local To Component(LocalPose=Use cached pose 'cachedPose'())
    Alpha=1
    Translation="2,0,2.5"
    Rotation="10,20,0"
    Scale="1,1,1"
   ) {
    rotationMode=BMM_Additive
    translationMode=BMM_Additive
    alphaBoolBlend {blendOption=Linear}
    boneToModify {boneName=dyn_pauldron_l}
    rotation {pitch=10, yaw=20}
    translation {x=2, z=2.5}
   }
   n6 = "Transform (Modify) Bone - Bone: dyn_pauldron_r"(ComponentPose=@n5, Alpha=1, Translation="-2,0,2.5", Rotation="-10,-20,0", Scale="1,1,1") {
    rotationMode=BMM_Additive
    translationMode=BMM_Additive
    alphaBoolBlend {blendOption=Linear}
    boneToModify {boneName=dyn_pauldron_r}
    rotation {pitch=-10, yaw=-20}
    translation {x=-2, z=2.5}
   }
   n7 = Component To Local(ComponentPose=@n6)
   n8 = OR Boolean(A=Get bIsSkyDiving(), B=Get bIsParachuteOpened())
   n9 = "Transform (Modify) Bone - Bone: dyn_pauldron_l"(ComponentPose=Local To Component(LocalPose=Use cached pose 'cachedPose'()), Alpha=1, Translation="3,0,3", Rotation="10,0,0", Scale="1,1,1") {
    rotationMode=BMM_Additive
    translationMode=BMM_Additive
    alphaBoolBlend {blendOption=Linear}
    boneToModify {boneName=dyn_pauldron_l}
    rotation {pitch=10}
    translation {x=3, z=3}
   }
   n10 = "Transform (Modify) Bone - Bone: dyn_pauldron_r"(ComponentPose=@n9, Alpha=1, Translation="-3,0,3", Rotation="-10,0,0", Scale="1,1,1") {
    rotationMode=BMM_Additive
    translationMode=BMM_Additive
    alphaBoolBlend {blendOption=Linear}
    boneToModify {boneName=dyn_pauldron_r}
    rotation {pitch=-10}
    translation {x=-3, z=3}
   }
   Output Pose(Result=Blend Poses by bool(
     BlendPose_0=Component To Local(ComponentPose=@n10)
     BlendPose_1=Blend Poses by bool(BlendPose_0=@n7, BlendPose_1=@n4, bActiveValue=@n8) {blendType=Linear}
     bActiveValue=Get bIsParachuteOpened()
    ) {blendType=Linear}) {layerGroup=None}
   n11 = RigidBody(ComponentPose=Local To Component(LocalPose=Copy Pose From Mesh() {bUseAttachedParent=true})) {
    bOverrideWorldGravity=true
    overridePhysicsAsset=&/F_MED_Relish_Physics
    simulationSpace=BaseBoneSpace
    alphaBoolBlend {blendOption=Linear}
    baseBoneRef {boneName=dyn_MSTR}
    componentAppliedLinearAccClamp=<3000 3000 3000>
    componentLinearAccScale=<0.2 0.2 0.4>
    componentLinearVelScale=<0.1 0.1 0.2>
    overrideWorldGravity {z=-980}
   }
   n12 = Get isCustomEmotePlaying() -> (isCustomEmotePlaying:double)
   n13 = "Transform (Modify) Bone - Bone: dyn_backstrap"(ComponentPose=@n11, Alpha=@n12, Translation="1.6,-1.5,0", Rotation="0,0,0", Scale="1,1,1") {
    rotationMode=BMM_Additive
    translationMode=BMM_Additive
    alphaBoolBlend {blendOption=Linear}
    boneToModify {boneName=dyn_backstrap}
    translation=<1.6 -1.5>
   }
   n14 = "Transform (Modify) Bone - Bone: dyn_pauldron_r_tweak"(ComponentPose=@n13, Alpha=@n12, Translation="0,0,0", Rotation="0,0,-10", Scale="1,1,1") {
    rotationMode=BMM_Additive
    translationMode=BMM_Additive
    alphaBoolBlend {blendOption=Linear}
    boneToModify {boneName=dyn_pauldron_r_tweak}
    rotation {roll=-10}
   }
   n15 = "Transform (Modify) Bone - Bone: dyn_arm_l_holster"(ComponentPose=@n14, Alpha=@n12, Translation="-0.3,0.25,0", Rotation="0,0,0", Scale="1,1,1") {
    rotationMode=BMM_Additive
    translationMode=BMM_Additive
    alphaBoolBlend {blendOption=Linear}
    boneToModify {boneName=dyn_arm_l_holster}
    translation=<-0.3 0.25>
   }
   Save cached pose 'cachedPose'(Pose=Component To Local(ComponentPose=@n15)) {cacheName=cachedPose}
  ]}
 EventGraph {nodes [
   Event BlueprintUpdateAnimation()
   n1 = GetAnimInstance(self=GetOwnerSkeletalMeshComponent())
   > Is Valid(InputObject=@n1)
   > Set isCustomEmotePlaying(isCustomEmotePlaying:double=SelectFloat(
     A=1
     B=0
     bPickA=Montage_IsActive!(
      self:Anim Instance Object Reference=@n1
      Montage:Anim Montage Object Reference=/BRCosmetics/Animation/Game/MainPlayer/Menu/BR/CustomSkins/_s18/Relish/Relish_Female_Idle_Montage
     ) -> (ReturnValue:bool)
    )) -> (Output_Get:double)
  ]}
]
variables [isCustomEmotePlaying {type=number}]
