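# Physics asset for the face accessory mesh of the F_MED_Relish player cosmetic, covering dynamic hair and beret bones.
# Capsule bodies form bang and back-hair chains constrained to the head, with tighter limits at roots and a fork in left bang B.
# The beret hangs directly off the head with asymmetric swing limits.
# Head and neck bodies form a constrained chain, while the right clavicle capsule has no constraint.
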PhysicsAsset /BRCosmetics/Characters/Player/Female/Medium/Bodies/F_MED_Relish/Meshes/Parts/F_MED_Relish_FaceAcc_Physics
bodies {
 clavicle_r [{
   shapeType=Capsule
   length=9.4701
   radius=5.7299
   shapeName=clavicle_r_capsule
   center=<-5.8758 0.9843 -1.0049>
   rotation=<rot 0 -89.9999 100.0003>
  }]
 dyn_back_a_01 [{shapeType=Capsule, length=1, radius=1, shapeName=dyn_back_a_01_capsule, rotation=<rot 0 0 90>}]
 dyn_back_a_02 [{shapeType=Capsule, length=2, radius=1, shapeName=dyn_back_a_02_capsule, center=<2.5 0 0.3>, rotation=<rot 0 0 90>}]
 dyn_back_a_03 [{shapeType=Capsule, length=2, radius=1, shapeName=dyn_back_a_03_capsule, center=<2.5 0 0.3>, rotation=<rot 0 0 90>}]
 dyn_back_b_01 [{shapeType=Capsule, length=1, radius=1, shapeName=dyn_back_b_01_capsule, rotation=<rot 0 0 90>}]
 dyn_back_b_02 [{shapeType=Capsule, length=2, radius=1, shapeName=dyn_back_b_02_capsule, center=<2.5 0 0.3>, rotation=<rot 0 0 90>}]
 dyn_back_b_03 [{shapeType=Capsule, length=2, radius=1, shapeName=dyn_back_b_03_capsule, center=<2.5 0 0.3>, rotation=<rot 0 0 90>}]
 dyn_back_c_01 [{shapeType=Capsule, length=1, radius=1, shapeName=dyn_back_c_01_capsule, rotation=<rot 0 0 90>}]
 dyn_back_c_02 [{shapeType=Capsule, length=2, radius=1, shapeName=dyn_back_c_02_capsule, center=<2.5 0 0.3>, rotation=<rot 0 0 90>}]
 dyn_back_c_03 [{shapeType=Capsule, length=2, radius=1, shapeName=dyn_back_c_03_capsule, center=<2.6627 0 -1.1441>, rotation=<rot 0 0 90>}]
 dyn_beret [{shapeType=Capsule, length=3, radius=2, shapeName=dyn_beret_capsule, center=<0.4041 0.0001 0>, rotation=<rot 0 0 90>}]
 dyn_l_bang_b_01 [{shapeType=Capsule, length=1, radius=1, shapeName=dyn_l_bang_b_01_capsule, rotation=<rot 0 0 90>}]
 dyn_l_bang_b_02 [{shapeType=Capsule, length=2, radius=1, shapeName=dyn_l_bang_b_02_capsule, center=<2.5 0 0.3>, rotation=<rot 0 0 90>}]
 dyn_l_bang_b_03 [{shapeType=Capsule, length=2, radius=1, shapeName=dyn_l_bang_b_03_capsule, center=<2.5 0 0.3>, rotation=<rot 0 0 90>}]
 dyn_l_bang_b_04 [{shapeType=Capsule, length=2, radius=1, shapeName=dyn_l_bang_b_04_capsule, center=<2.5 0 0.3>, rotation=<rot 0 0 90>}]
 dyn_l_bang_b_05 [{shapeType=Capsule, length=2, radius=1, shapeName=dyn_l_bang_b_05_capsule, center=<2.5 0 0.3>, rotation=<rot 0 0 90>}]
 dyn_l_bang_b_06 [{shapeType=Capsule, length=2, radius=1, shapeName=dyn_l_bang_b_06_capsule, center=<2.5 0 0.3>, rotation=<rot 0 0 90>}]
 dyn_l_bang_b_07 [{shapeType=Capsule, length=2, radius=1, shapeName=dyn_l_bang_b_07_capsule, center=<2.5 0 0.3>, rotation=<rot 0 0 90>}]
 dyn_l_bang_d_01 [{shapeType=Capsule, length=1, radius=1, shapeName=dyn_l_bang_d_01_capsule, rotation=<rot 0 0 90>}]
 dyn_l_bang_d_02 [{shapeType=Capsule, length=2, radius=1, shapeName=dyn_l_bang_d_02_capsule, center=<2.5 0 0.3>, rotation=<rot 0 0 90>}]
 dyn_l_bang_d_03 [{shapeType=Capsule, length=2, radius=1, shapeName=dyn_l_bang_d_03_capsule, center=<2.5 0 0.3>, rotation=<rot 0 0 90>}]
 dyn_l_bang_d_04 [{shapeType=Capsule, length=2, radius=1, shapeName=dyn_l_bang_d_04_capsule, center=<2.5 0 0.3>, rotation=<rot 0 0 90>}]
 dyn_r_bang_a_01 [{shapeType=Capsule, length=1, radius=1, shapeName=dyn_r_bang_a_01_capsule, rotation=<rot 0 0 90>}]
 dyn_r_bang_a_02 [{shapeType=Capsule, length=2, radius=1, shapeName=dyn_r_bang_a_02_capsule, center=<2.0092 0 0.3>, rotation=<rot 0 0 90>}]
 dyn_r_bang_a_03 [{shapeType=Capsule, length=2, radius=1, shapeName=dyn_r_bang_a_03_capsule, center=<2.0092 0 0.3>, rotation=<rot 0 0 90>}]
 dyn_r_bang_b_01 [{shapeType=Capsule, length=1, radius=1, shapeName=dyn_r_bang_b_01_capsule, rotation=<rot 0 0 90>}]
 dyn_r_bang_b_02 [{shapeType=Capsule, length=2, radius=1, shapeName=dyn_r_bang_b_02_capsule, center=<1.5937 0 0.3>, rotation=<rot 0 0 90>}]
 dyn_r_bang_b_03 [{shapeType=Capsule, length=2, radius=1, shapeName=dyn_r_bang_b_03_capsule, center=<1.2269 -0.922 0.0131>, rotation=<rot 0 24.9967 74.9976>}]
 head [
  {shapeType=Sphere, radius=7.6298, shapeName=head_sphere, center=<1.1626 1.7476 1.3538>}
  {shapeType=Capsule, length=5, radius=7, shapeName=head_capsule, center=<3.4316 0.0001 0>, rotation=<rot 90 0 0>}
 ]
 neck_01 [{shapeType=Capsule, length=5, radius=4, shapeName=neck_01_capsule}]
 neck_02 [{shapeType=Capsule, length=3.8699, radius=5.1301, shapeName=neck_02_capsule, center=<1.1763 -3.4833 0>, rotation=<rot 90 0 0>}]
}
constraints [
 {bone1Name=neck_02, bone2Name=neck_01, swing1LimitDegrees=10, swing2LimitDegrees=10, twistLimitDegrees=1}
 {bone1Name=head, bone2Name=neck_02, swing1LimitDegrees=10, swing2LimitDegrees=10, twistLimitDegrees=1}
 {bone1Name=dyn_r_bang_b_01, bone2Name=head, swing1LimitDegrees=10, swing2LimitDegrees=10, twistLimitDegrees=1}
 {bone1Name=dyn_r_bang_b_02, bone2Name=dyn_r_bang_b_01, swing1LimitDegrees=20, swing2LimitDegrees=20, twistLimitDegrees=10}
 {bone1Name=dyn_r_bang_b_03, bone2Name=dyn_r_bang_b_02, swing1LimitDegrees=20, swing2LimitDegrees=35, twistLimitDegrees=10}
 {bone1Name=dyn_r_bang_a_01, bone2Name=head, swing1LimitDegrees=10, swing2LimitDegrees=10, twistLimitDegrees=1}
 {bone1Name=dyn_r_bang_a_02, bone2Name=dyn_r_bang_a_01, swing1LimitDegrees=20, swing2LimitDegrees=35, twistLimitDegrees=10}
 {bone1Name=dyn_r_bang_a_03, bone2Name=dyn_r_bang_a_02, swing1LimitDegrees=20, swing2LimitDegrees=35, twistLimitDegrees=10}
 {bone1Name=dyn_l_bang_b_01, bone2Name=head, swing1LimitDegrees=10, swing2LimitDegrees=10, twistLimitDegrees=1}
 {bone1Name=dyn_l_bang_b_02, bone2Name=dyn_l_bang_b_01, swing1LimitDegrees=20, swing2LimitDegrees=35, twistLimitDegrees=10}
 {bone1Name=dyn_l_bang_b_03, bone2Name=dyn_l_bang_b_02, swing1LimitDegrees=20, swing2LimitDegrees=35, twistLimitDegrees=10}
 {bone1Name=dyn_l_bang_b_04, bone2Name=dyn_l_bang_b_03, swing1LimitDegrees=20, swing2LimitDegrees=35, twistLimitDegrees=10}
 {bone1Name=dyn_l_bang_b_06, bone2Name=dyn_l_bang_b_02, swing1LimitDegrees=20, swing2LimitDegrees=35, twistLimitDegrees=10}
 {bone1Name=dyn_l_bang_b_07, bone2Name=dyn_l_bang_b_06, swing1LimitDegrees=20, swing2LimitDegrees=35, twistLimitDegrees=10}
 {bone1Name=dyn_l_bang_b_05, bone2Name=dyn_l_bang_b_02, swing1LimitDegrees=20, swing2LimitDegrees=35, twistLimitDegrees=10}
 {bone1Name=dyn_l_bang_d_01, bone2Name=head, swing1LimitDegrees=10, swing2LimitDegrees=10, twistLimitDegrees=1}
 {bone1Name=dyn_l_bang_d_02, bone2Name=dyn_l_bang_d_01, swing1LimitDegrees=20, swing2LimitDegrees=35, twistLimitDegrees=10}
 {bone1Name=dyn_l_bang_d_03, bone2Name=dyn_l_bang_d_02, swing1LimitDegrees=20, swing2LimitDegrees=35, twistLimitDegrees=10}
 {bone1Name=dyn_l_bang_d_04, bone2Name=dyn_l_bang_d_03, swing1LimitDegrees=20, swing2LimitDegrees=35, twistLimitDegrees=10}
 {bone1Name=dyn_back_a_01, bone2Name=head, swing1LimitDegrees=10, swing2LimitDegrees=10, twistLimitDegrees=1}
 {bone1Name=dyn_back_a_02, bone2Name=dyn_back_a_01, swing1LimitDegrees=20, swing2LimitDegrees=35, twistLimitDegrees=10}
 {bone1Name=dyn_back_a_03, bone2Name=dyn_back_a_02, swing1LimitDegrees=20, swing2LimitDegrees=35, twistLimitDegrees=10}
 {bone1Name=dyn_back_b_01, bone2Name=head, swing1LimitDegrees=10, swing2LimitDegrees=10, twistLimitDegrees=1}
 {bone1Name=dyn_back_b_02, bone2Name=dyn_back_b_01, swing1LimitDegrees=20, swing2LimitDegrees=35, twistLimitDegrees=10}
 {bone1Name=dyn_back_b_03, bone2Name=dyn_back_b_02, swing1LimitDegrees=20, swing2LimitDegrees=35, twistLimitDegrees=10}
 {bone1Name=dyn_back_c_01, bone2Name=head, swing1LimitDegrees=10, swing2LimitDegrees=10, twistLimitDegrees=1}
 {bone1Name=dyn_back_c_02, bone2Name=dyn_back_c_01, swing1LimitDegrees=20, swing2LimitDegrees=35, twistLimitDegrees=10}
 {bone1Name=dyn_back_c_03, bone2Name=dyn_back_c_02, swing1LimitDegrees=20, swing2LimitDegrees=20, twistLimitDegrees=10}
 {bone1Name=dyn_beret, bone2Name=head, swing1LimitDegrees=2, swing2LimitDegrees=20, twistLimitDegrees=1}
]
